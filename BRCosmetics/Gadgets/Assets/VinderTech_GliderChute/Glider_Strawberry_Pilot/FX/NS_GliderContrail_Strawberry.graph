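# Niagara contrail effect for the Strawberry Pilot glider, combining cyan engine glow, wind and wingtip trails.
# Looping burst particles pin engine meshes and flare sprites to seven skeletal mesh sockets, with a fire sprite stream and flickering light at socket 3.
# Wind meshes and two AnimTrail ribbons follow the skeleton.
# A system FN_Integrator value drives engine brightness and stretch plus wind and trail opacity, while front-end modifiers adjust engine and fire placement.
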NiagaraSystem /BRCosmetics/Gadgets/Assets/VinderTech_GliderChute/Glider_Strawberry_Pilot/FX/NS_GliderContrail_Strawberry
dependencies {
 usedDataInterfaces [&NiagaraDataInterfaceVectorCurve, &NiagaraDataInterfaceCurve, &NiagaraDataInterfaceArrayFloat3]
 usedDynamicInputs []
 usedModules [&SystemState, &FN_Integrator]
}
emitters {
 Engines {
  emitterUpdateScript [
   &EmitterState(Inactive Response=::NewEnumerator1, Life Cycle Mode=::NewEnumerator1, Loop Duration=1, UseLoopDelay=true)
   &SpawnBurst_Instantaneous(Spawn Count=7, Spawn Probability=HlslExpression() {hlslExpression="Emitter.LoopCount < 1"}, Use Spawn Probability=true)
  ]
  particleSpawnScript [
   /SetVariables(Particles.SpriteRotationRate=&RandomRangeFloat(Minimum=-0.05, Maximum=0.05, Evaluation Type=::NewEnumerator1)) {isSetParameters=true, module=SetVariables_#1}
   &FN_SetSpawnGroup(Number of Spawn Groups=7)
   /SetVariables(Particles.VisibilityTag=&SpawnGroup()) {isSetParameters=true}
   /Niagara/Modules/Spawn/Initialization/InitializeParticle(
    Lifetime=&RandomRangeFloat(Minimum=1.1, Maximum=1.2, Evaluation Type=::NewEnumerator1)
    Mesh Scale=&FN_FrontEndVector3Modifier(Gameplay=<1 1 1>, Front End=<1 1.5 1>)
    Module.Write Color=true
    Module.Write Lifetime=true
    Module.Write Mass=true
    Module.Write Material Random=true
    Module.Write Position=false
    Module.Write Scale=true
    Module.Write SpriteRotation=false
    Module.Write SpriteSize=true
    Position=<0 0 0>
    Sprite Size=&Vector2DFromFloat(Value=&RandomRangeFloat(Minimum=30, Maximum=40, Evaluation Type=::NewEnumerator1))
   )
   &Color(Color=&MakeLinearColorFromVectorAndFloat(
     "Vector (RGB)"=&RandomRangeVector(Minimum=<0 1.5 1.2146>, Maximum=<0 1 2>, Evaluation Type=::NewEnumerator1)
     "Float (Alpha)"=&ClampFloat(Float=0.35)
    ))
   &InitialMeshOrientation(
    Orientation Vector=&FN_FrontEndVector3Modifier(Gameplay=<1 0 0>, Front End=<-1 0 0>)
    Use Orientation Vector=true
    Use Rotation Vector=false
   )
  ]
  particleUpdateScript [
   &ParticleState(
    Kill Particles When Lifetime Has Elapsed=false
    Let Infinitely Lived Particles Die When Emitter Deactivates=true
    Loop Particles Lifetime=true
   )
   &ScaleColor(
    Scale Alpha=&FloatFromCurve(
     FloatCurve=DataInterface() {propertyValues {Curve {keys [
         <at 0.0017 0.5 RCIM_Cubic>
         {arriveTangent=0.6275, interpMode=RCIM_Cubic, leaveTangent=0.6275, time=0.0848, value=0.9873}
         {arriveTangent=-0.5342, interpMode=RCIM_Cubic, leaveTangent=-0.5342, time=0.7975, value=0.9994}
         <at 0.997 0.5 RCIM_Cubic>
        ]}}}
     Scale Curve=30
    )
    Scale RGB=&VectorFromFloat(Value=&Lerp_Float(A=1, B=1.5, Alpha=$System.FN_Integrator001.Value))
    ScaleA=true
    ScaleRGB=true
   )
   &SkeletalMeshLocation(
    Bone / Socket Index=$Particles.SpawnGroup
    Inherit Velocity=false
    MaskSpawnGroup=false
    Mesh Sampling Type=::NewEnumerator5
    Sampled Position Offset=&SelectVectorFromArray(
     Direct Array Index=$Particles.SpawnGroup
     Vector Selection Array=DataInterface() {propertyValues {FloatData [<0 0 0>, <0 0 0>, <0 0 0>, <0 0 -5>, <0 0 0>, <0 0 0>, <0 0 0>]}}
    )
    Skeletal Mesh=$User.SKM_Engines
    Socket Sampling Mode=::NewEnumerator5
   )
   &ScaleSpriteSize(Scale Factor=&Vector2DFromFloat(Value=&FloatFromCurve(FloatCurve=DataInterface() {propertyValues {Curve {keys [<at 0 1 RCIM_Cubic>,<at 0.5 1.25 RCIM_Cubic>,<at 1 1 RCIM_Cubic>]}}})))
   &DynamicMaterialParameters(
    Index 0 Param 1=1
    Index 0 Param 2=1
    Index 0 Param 3=1
    Index 0 Param 4=1
    Module.Param0WriteEnabled=true
    Module.Param1WriteEnabled=true
    Module.Param2WriteEnabled=true
    Module.Param3WriteEnabled=true
   )
   &ScaleMeshSize(Scale Factor=&Lerp_Vector(A=<1 1 1>, B=<1 1.75 1>, Alpha=$System.FN_Integrator001.Value))
  ]
  rendererClasses [&NiagaraMeshRendererProperties]
  renderers [
   {rendererClass=&NiagaraMeshRendererProperties, rendererIndex=0}
   {rendererClass=&NiagaraMeshRendererProperties, rendererIndex=1}
   {rendererClass=&NiagaraMeshRendererProperties, rendererIndex=2}
   {rendererClass=&NiagaraMeshRendererProperties, rendererIndex=3}
   {rendererClass=&NiagaraMeshRendererProperties, rendererIndex=4}
   {rendererClass=&NiagaraMeshRendererProperties, rendererIndex=5}
   {rendererClass=&NiagaraMeshRendererProperties, rendererIndex=6}
  ]
 }
 FireBig {
  emitterUpdateScript [
   &EmitterState(Inactive Response=::NewEnumerator1, Life Cycle Mode=::NewEnumerator1, Loop Duration=1, UseLoopDelay=true)
   &SpawnRate(SpawnRate=10, Use Spawn Probability=false)
  ]
  particleSpawnScript [
   &InitializeParticle(
    Lifetime=&RandomRangeFloat(Minimum=0.4, Maximum=0.5, Evaluation Type=::NewEnumerator1)
    Module.Write Color=true
    Module.Write Lifetime=true
    Module.Write Mass=true
    Module.Write Material Random=true
    Module.Write Position=false
    Module.Write Scale=false
    Module.Write SpriteRotation=true
    Module.Write SpriteSize=true
    Sprite Rotation=&NormalizedAngleToDegrees(NormalizedAngle=&RandomRangeFloat(Minimum=0, Maximum=1, Evaluation Type=::NewEnumerator1))
    Sprite Size=&Vector2DFromFloat(Value=&Lerp_Float(
      A=&RandomRangeFloat(Minimum=20, Maximum=25, Evaluation Type=::NewEnumerator1)
      B=&RandomRangeFloat(Minimum=10, Maximum=15)
      Alpha=&MakeCustomFloatFromBool(Bool=$User.bIsFrontEndPreview, True Float=1, False Float=0)
     ))
   )
   &Color(Color=&MakeLinearColorFromVectorAndFloat("Vector (RGB)"=<1 1 1>, "Float (Alpha)"=&ClampFloat(Float=0.1)))
  ]
  particleUpdateScript [
   &ParticleState()
   &DynamicMaterialParameters(
    Index 0 Param 1=&FloatFromCurve(FloatCurve=DataInterface() {propertyValues {Curve {keys [<at 0.5 1 RCIM_Cubic>,<at 1 -1>]}}})
    Index 0 Param 2=1
    Index 0 Param 3=1
    Index 0 Param 4=1
    Module.Param0WriteEnabled=true
    Module.Param1WriteEnabled=true
    Module.Param2WriteEnabled=true
    Module.Param3WriteEnabled=true
   )
   &SkeletalMeshLocation(
    Bone / Socket Index=3
    Inherit Velocity=false
    Interpolate Mesh Sampling=false
    MaskSpawnGroup=false
    Mesh Sampling Type=::NewEnumerator5
    Sampled Position Offset=&Lerp_Vector(
     A=<0 -10 -6>
     B=<0 -5 -3.5>
     Alpha=&MakeCustomFloatFromBool(Bool=$User.bIsFrontEndPreview, True Float=1, False Float=0)
    )
    Sampled Position Offset Coordinate Space:ENiagaraMeshTransforms=::NewEnumerator3
    Skeletal Mesh=$User.SKM_Engines
    Socket Sampling Mode=::NewEnumerator5
   )
   &SubUVAnimation(Sub UV Play Rate=0.1, SubUV Animation Mode:ENiagara_SubUVLookupMode=::NewEnumerator3)
   &ScaleSpriteSize(Scale Factor=&Vector2DFromFloat(Value=&FloatFromCurve(FloatCurve=DataInterface() {propertyValues {Curve {keys [{arriveTangent=1.4511, interpMode=RCIM_Cubic, time=-0.0013, value=0.1965}, <at 1 1.25>]}}})))
   &ScaleColor(
    Scale Alpha=&FloatFromCurve(FloatCurve=DataInterface() {propertyValues {Curve {keys [
         <at -0.0022 -0.0013 RCIM_Cubic>
         {arriveTangent=1.6555, interpMode=RCIM_Cubic, leaveTangent=1.6555, time=0.1497, value=-0.0095}
         {arriveTangent=0.7347, interpMode=RCIM_Cubic, leaveTangent=0.7347, time=0.5987, value=0.9935}
         {arriveTangent=-2.4759, interpMode=RCIM_Cubic, leaveTangent=-2.4759, time=0.856, value=0.5095}
         <at 1 0>
        ]}}})
    Scale RGB=&VectorFromCurve(VectorCurve=DataInterface() {propertyValues {
       XCurve {keys [{arriveTangent=-861.5734, interpMode=RCIM_Cubic, time=0, value=0}, <at 0.2668 0 RCIM_Cubic>]}
       YCurve {keys [{arriveTangent=-66.82, interpMode=RCIM_Cubic, time=0, value=182.4842}, <at 0.2668 109.4029 RCIM_Cubic>]}
       ZCurve {keys [{arriveTangent=-0.0001, interpMode=RCIM_Cubic, time=0, value=250}, <at 0.2668 150 RCIM_Cubic>]}
      }})
    ScaleA=true
    ScaleRGB=true
   )
  ]
  rendererClasses [&NiagaraSpriteRendererProperties]
  renderers [{rendererClass=&NiagaraSpriteRendererProperties}]
 }
 Flares {
  emitterUpdateScript [
   &EmitterState(Inactive Response=::NewEnumerator1, Life Cycle Mode=::NewEnumerator1, Loop Duration=1, UseLoopDelay=true)
   &SpawnBurst_Instantaneous(Spawn Count=7, Spawn Probability=HlslExpression() {hlslExpression="Emitter.LoopCount < 1"}, Use Spawn Probability=true)
  ]
  particleSpawnScript [
   /SetVariables(Particles.SpriteRotationRate=&RandomRangeFloat(Minimum=-0.05, Maximum=0.05, Evaluation Type=::NewEnumerator1)) {isSetParameters=true, module=SetVariables_#1}
   &FN_SetSpawnGroup(Number of Spawn Groups=7)
   &InitializeParticle(
    Lifetime=&RandomRangeFloat(Minimum=1.1, Maximum=1.2, Evaluation Type=::NewEnumerator1)
    Module.Write Color=true
    Module.Write Lifetime=true
    Module.Write Mass=true
    Module.Write Material Random=true
    Module.Write Position=false
    Module.Write Scale=false
    Module.Write SpriteRotation=false
    Module.Write SpriteSize=true
    Position=<0 0 0>
    Sprite Size=&Vector2DFromFloat(Value=&RandomRangeFloat(Minimum=30, Maximum=40, Evaluation Type=::NewEnumerator1))
   )
   &Color(Color=&MakeLinearColorFromVectorAndFloat(
     "Vector (RGB)"=&RandomRangeVector(Minimum=<0 1.5 1.2146>, Maximum=<0 1 2>, Evaluation Type=::NewEnumerator1)
     "Float (Alpha)"=&ClampFloat(Float=0.35)
    ))
   /SetVariables(Particles.Color_02_Init:LinearColor=&MakeLinearColorFromVectorAndFloat(
     "Vector (RGB)"=&RandomRangeVector(Minimum=<0 1.4708 1.5>, Maximum=<0 1 2>, Evaluation Type=::NewEnumerator1)
     "Float (Alpha)"=&ClampFloat(Float=0.25)
    )) {isSetParameters=true}
  ]
  particleUpdateScript [
   &ParticleState(
    Kill Particles When Lifetime Has Elapsed=false
    Let Infinitely Lived Particles Die When Emitter Deactivates=true
    Loop Particles Lifetime=true
   )
   &ScaleColor(
    Scale Alpha=&FloatFromCurve(
     FloatCurve=DataInterface() {propertyValues {Curve {keys [
         <at 0.0017 0.5 RCIM_Cubic>
         {arriveTangent=0.6275, interpMode=RCIM_Cubic, leaveTangent=0.6275, time=0.0848, value=0.9873}
         {arriveTangent=-0.5342, interpMode=RCIM_Cubic, leaveTangent=-0.5342, time=0.7975, value=0.9994}
         <at 0.997 0.5 RCIM_Cubic>
        ]}}}
     Scale Curve=0.35
    )
    ScaleA=true
    ScaleRGB=true
   )
   &SkeletalMeshLocation(
    Bone / Socket Index=$Particles.SpawnGroup
    Inherit Velocity=false
    MaskSpawnGroup=false
    Mesh Sampling Type=::NewEnumerator5
    Sampled Position Offset=&SelectVectorFromArray(
     Direct Array Index=$Particles.SpawnGroup
     Vector Selection Array=DataInterface() {propertyValues {FloatData [<0 0 0>, <0 0 0>, <0 -7 0>, <0 0 -5>, <0 0 0>, <0 0 0>, <0 -7 0>]}}
    )
    Sampled Position Offset Coordinate Space:ENiagaraMeshTransforms=::NewEnumerator3
    Skeletal Mesh=$User.SKM_Engines
    Socket Sampling Mode=::NewEnumerator5
   )
   &ScaleSpriteSize(Scale Factor=&Vector2DFromFloat(Value=&FloatFromCurve(FloatCurve=DataInterface() {propertyValues {Curve {keys [<at 0 1 RCIM_Cubic>,<at 0.5 1.25 RCIM_Cubic>,<at 1 1 RCIM_Cubic>]}}})))
   &DynamicMaterialParameters(
    Index 0 Param 1=1
    Index 0 Param 2=1
    Index 0 Param 3=1
    Index 0 Param 4=1
    Module.Param0WriteEnabled=true
    Module.Param1WriteEnabled=true
    Module.Param2WriteEnabled=true
    Module.Param3WriteEnabled=true
   )
   /SetVariables(
    Particles.Color_02:LinearColor=&MultiplyLinearColor(
     Linear Color=$Particles.Initial.Color_02_Init
     Scale Factor=&MakeFloat4FromVectorAndFloat(
      "Vector (RGB)"=<1 1 1>
      "Float (Alpha)"=&FloatFromCurve(
       FloatCurve=DataInterface() {propertyValues {Curve {keys [{arriveTangent=-0.0037, interpMode=RCIM_Cubic, time=-0.0007, value=0.5}, <at 0.4381 1.0043 RCIM_Cubic>, <at 0.997 0.5 RCIM_Cubic>]}}}
       Scale Curve=10
      )
     )
    )
    Particles.SpriteSize_02:Vector2f=&Vector2DFromFloat(Value=&RandomRangeFloat(Minimum=50, Maximum=60))
   ) {isSetParameters=true}
   &SpriteRotationRate(Rotation Rate=&RandomRangeFloat(Minimum=10, Maximum=-10)) {module=SpriteRotationRate001}
  ]
  rendererClasses [&NiagaraSpriteRendererProperties]
  renderers [
   {rendererClass=&NiagaraSpriteRendererProperties}
   {rendererClass=&NiagaraSpriteRendererProperties, rendererIndex=1}
  ]
 }
 Light {
  emitterUpdateScript [
   &EmitterState(Inactive Response=::NewEnumerator1, Life Cycle Mode=::NewEnumerator1, Loop Duration=1, UseLoopDelay=true)
   &SpawnBurst_Instantaneous(Spawn Count=1, Spawn Probability=HlslExpression() {hlslExpression="Emitter.LoopCount < 1"}, Use Spawn Probability=true)
  ]
  particleSpawnScript [
   /Niagara/Modules/Spawn/Initialization/V2/InitializeParticle(
    Lifetime Max=0.5
    Lifetime Min=0.4
    Lifetime Mode=::NewEnumerator1
    Sprite Rotation Mode=::NewEnumerator2
    Sprite Size Mode=::NewEnumerator1
    Uniform Sprite Size Max=25
    Uniform Sprite Size Min=20
    UsePositionOffset=false
    Write Material Random=true
   ) {module=InitializeParticle001}
   &Color(Color=&MakeLinearColorFromVectorAndFloat("Vector (RGB)"=<1 1 1>, "Float (Alpha)"=&ClampFloat(Float=0.1)))
  ]
  particleUpdateScript [
   &ParticleState(
    Kill Particles When Lifetime Has Elapsed=false
    Let Infinitely Lived Particles Die When Emitter Deactivates=true
    Loop Particles Lifetime=true
   )
   &DynamicMaterialParameters(
    Index 0 Param 1=&FloatFromCurve(FloatCurve=DataInterface() {propertyValues {Curve {keys [<at 0.5 1 RCIM_Cubic>,<at 1 -1>]}}})
    Index 0 Param 2=1
    Index 0 Param 3=1
    Index 0 Param 4=1
    Module.Param0WriteEnabled=true
    Module.Param1WriteEnabled=true
    Module.Param2WriteEnabled=true
    Module.Param3WriteEnabled=true
   )
   &SkeletalMeshLocation(
    Bone / Socket Index=3
    Inherit Velocity=false
    Interpolate Mesh Sampling=false
    MaskSpawnGroup=false
    Mesh Sampling Type=::NewEnumerator5
    Sampled Position Offset=<0 -10 -6>
    Sampled Position Offset Coordinate Space:ENiagaraMeshTransforms=::NewEnumerator3
    Skeletal Mesh=$User.SKM_Engines
    Socket Sampling Mode=::NewEnumerator5
   )
   &SubUVAnimation(Sub UV Play Rate=0.1, SubUV Animation Mode:ENiagara_SubUVLookupMode=::NewEnumerator3)
   &ScaleSpriteSize(Scale Factor=&Vector2DFromFloat(Value=&FloatFromCurve(FloatCurve=DataInterface() {propertyValues {Curve {keys [{arriveTangent=1.4511, interpMode=RCIM_Cubic, time=-0.0013, value=0.1965}, <at 1 1.25>]}}})))
   &ScaleColor(
    Scale Alpha=&FloatFromCurve(FloatCurve=DataInterface() {propertyValues {Curve {keys [
         <at -0.0022 -0.0013 RCIM_Cubic>
         {arriveTangent=1.6555, interpMode=RCIM_Cubic, leaveTangent=1.6555, time=0.1497, value=-0.0095}
         {arriveTangent=0.7347, interpMode=RCIM_Cubic, leaveTangent=0.7347, time=0.5987, value=0.9935}
         {arriveTangent=-2.4759, interpMode=RCIM_Cubic, leaveTangent=-2.4759, time=0.856, value=0.5095}
         <at 1 0>
        ]}}})
    Scale RGB=&VectorFromCurve(VectorCurve=DataInterface() {propertyValues {
       XCurve {keys [{arriveTangent=-861.5734, interpMode=RCIM_Cubic, time=0, value=0}, <at 0.2668 0 RCIM_Cubic>]}
       YCurve {keys [{arriveTangent=-66.82, interpMode=RCIM_Cubic, time=0, value=182.4842}, <at 0.2668 109.4029 RCIM_Cubic>]}
       ZCurve {keys [{arriveTangent=-0.0001, interpMode=RCIM_Cubic, time=0, value=250}, <at 0.2668 150 RCIM_Cubic>]}
      }})
    ScaleA=true
    ScaleRGB=true
   )
   &FN_Light_Attributes(
    Color=&MakeLinearColorFromVectorAndFloat("Vector (RGB)"=&RandomRangeVector(Minimum=<0 1.5 1.2146>, Maximum=<0 1 2>), "Float (Alpha)"=1)
    Create Color Attribute=true
    Create Exponent Attribute=true
    Create Radius Attribute=true
    Enable Particle Light Enable=true
    Exponent=4
    Light Enabled=true
    Radius=90
    Scale Color=&FloatFromCurve(
     FloatCurve=DataInterface() {propertyValues {Curve {keys [
         <at 0 1 RCIM_Cubic>
         {arriveTangent=-0.2299, interpMode=RCIM_Cubic, leaveTangent=-0.2299, time=0.0743, value=1.2496}
         {arriveTangent=-0.4635, interpMode=RCIM_Cubic, leaveTangent=-0.4635, time=0.2645, value=0.9392}
         {arriveTangent=-1.0603, interpMode=RCIM_Cubic, leaveTangent=-1.0603, time=0.3571, value=1.1185}
         {arriveTangent=0.6715, interpMode=RCIM_Cubic, leaveTangent=0.6715, time=0.4997, value=0.6898}
         {arriveTangent=0.6619, interpMode=RCIM_Cubic, leaveTangent=0.6619, time=0.6447, value=1.3117}
         {arriveTangent=-0.8773, interpMode=RCIM_Cubic, leaveTangent=-0.8773, time=0.853, value=0.9237}
         <at 1 1 RCIM_Cubic>
        ]}}}
     Scale Curve=4
    )
   )
  ]
  rendererClasses [&NiagaraLightRendererProperties]
  renderers [{rendererClass=&NiagaraLightRendererProperties, rendererIndex=0}]
 }
 Wind {
  emitterUpdateScript [
   &EmitterState(Inactive Response=::NewEnumerator1, Life Cycle Mode=::NewEnumerator1, Loop Duration=1, UseLoopDelay=true)
   &SpawnBurst_Instantaneous(Spawn Count=8, Spawn Probability=HlslExpression() {hlslExpression="Emitter.LoopCount < 1"}, Use Spawn Probability=true)
  ]
  particleSpawnScript [
   /SetVariables(Particles.MeshOrientation=&MakeQuatFromAxisAngle(Axis=<-0.02 0.02 0>, AngleInDegrees=0)) {isSetParameters=true}
   &FN_SetSpawnGroup(Number of Spawn Groups=2)
   /SetVariables(Particles.VisibilityTag=$Particles.SpawnGroup) {isSetParameters=true}
   /Niagara/Modules/Spawn/Initialization/InitializeParticle(
    Color=<rgba 5 5 5 1>
    Lifetime=&RandomRangeFloat(Minimum=0.6, Maximum=1, Evaluation Type=::NewEnumerator1)
    Mesh Scale=<0.9 1 1>
    Module.Write Color=true
    Module.Write Lifetime=true
    Module.Write Mass=true
    Module.Write Material Random=true
    Module.Write Position=false
    Module.Write Scale=true
    Module.Write SpriteRotation=false
    Module.Write SpriteSize=false
    Position=<-140 -12 17>
   )
   &DynamicMaterialParameters(
    Index 0 Param 1=1
    Module.Param0WriteEnabled=true
    Module.Param1WriteEnabled=false
    Module.Param2WriteEnabled=false
    Module.Param3WriteEnabled=false
   )
   &InitialMeshOrientation(
    Rotation=&Lerp_Vector(A=<0 0 0>, B=<0 0.5 0>, Alpha=&MakeFloatFromInt(Int=$Particles.SpawnGroup))
    Use Orientation Vector=false
    Use Rotation Vector=true
   )
   &InitialMeshOrientation(
    Rotation=&Lerp_Vector(A=<0 0 0>, B=<0 0 0.5>, Alpha=&MakeFloatFromInt(Int=$Particles.SpawnGroup))
    Use Orientation Vector=false
    Use Rotation Vector=true
   ) {module=InitialMeshOrientation001}
  ]
  particleUpdateScript [
   &ParticleState(
    Kill Particles When Lifetime Has Elapsed=false
    Let Infinitely Lived Particles Die When Emitter Deactivates=true
    Loop Particles Lifetime=true
   )
   &ScaleColor(
    Scale Alpha=&FloatFromCurve(
     FloatCurve=DataInterface() {propertyValues {Curve {keys [<at 0 0>,<at 0.5 1>,<at 1 0>]}}}
     Scale Curve=&Lerp_Float(A=0, B=1, Alpha=$System.FN_Integrator001.Value)
    )
    Scale RGB=&VectorFromCurve(VectorCurve=DataInterface() {propertyValues {XCurve {keys [<at 0 1>,<at 1 0>]}, YCurve {keys [<at 0 1>,<at 1 0>]}, ZCurve {keys [<at 0 1>,<at 1 0>]}}})
    ScaleA=true
    ScaleRGB=true
   )
   &SkeletalMeshLocation(
    Apply Sampled Orientation to Mesh Particles=true
    Bone / Socket Index=$Particles.SpawnGroup
    Bone Sampling Mode=::NewEnumerator4
    Inherit Velocity=false
    MaskSpawnGroup=false
    Sampled Position Offset=&Lerp_Vector(A=<70 8 10>, B=<-70 -8 -10>, Alpha=&MakeFloatFromInt(Int=$Particles.SpawnGroup))
    Sampled Position Offset Coordinate Space:ENiagaraMeshTransforms=::NewEnumerator3
    Skeletal Mesh=$User.SKM_Wind
   )
  ]
  rendererClasses [&NiagaraMeshRendererProperties]
  renderers [
   {rendererClass=&NiagaraMeshRendererProperties, rendererIndex=0}
   {rendererClass=&NiagaraMeshRendererProperties, rendererIndex=1}
  ]
 }
 WingTrail {
  emitterSpawnScript [&FN_RibbonIDOffsetUtility()]
  emitterUpdateScript [
   &EmitterState()
   &FN_SpawnAnimTrail(Number of Anim Trails=2)
  ]
  particleSpawnScript [
   /SetVariables(Particles.AlphaFadeIn:NiagaraFloat=/Niagara/DynamicInputs/Math/Min_Float(A:NiagaraFloat=&Divide_Float(A=$Emitter.Age, B=1), B:NiagaraFloat=1)) {isSetParameters=true}
   &InitializeParticle(
    Lifetime=0.4
    Module.Write Color=true
    Module.Write Lifetime=true
    Module.Write Mass=true
    Module.Write Material Random=true
    Module.Write Position=false
    Module.Write Scale=false
    Module.Write SpriteRotation=false
    Module.Write SpriteSize=false
   )
   &Color(Color=&MakeLinearColorFromVectorAndFloat("Vector (RGB)"=<1 1 1>, "Float (Alpha)"=&ClampFloat(Float=0.55)))
   &FN_InitializeAnimTrailBoneIndecies(First Bone or Socket Index=1, Second Bone or Socket Index=0)
   &FN_InitializeAnimTrailBoneIndecies(AnimTrail Selection=1, First Bone or Socket Index=3, Second Bone or Socket Index=2) {module=FN_InitializeAnimTrailBoneIndecies001}
   &FN_InitializeAnimTrail(
    Color=&MakeLinearColor(
     R:NiagaraFloat=1
     G:NiagaraFloat=1
     B:NiagaraFloat=1
     A:NiagaraFloat=&Multiply_Float(A=$Particles.AlphaFadeIn, B=0.5)
    )
    Lifetime=&MakeCustomFloatFromBool(Bool=$Particles.Locked, True Float=100, False Float=1.5)
   )
  ]
  particleUpdateScript [
   &ParticleState()
   &ScaleColor(
    Scale Alpha=&FloatFromCurve(FloatCurve=DataInterface() {propertyValues {Curve {keys [
         <at 0 0.2 RCIM_Cubic>
         {arriveTangent=0.0505, interpMode=RCIM_Cubic, leaveTangent=0.0505, time=0.1871, value=1.0162}
         {arriveTangent=-1.2501, interpMode=RCIM_Cubic, leaveTangent=-1.2501, time=0.3318, value=0.2168}
         <at 1 0 RCIM_Cubic>
        ]}}})
    Scale RGB=&VectorFromCurve(VectorCurve=DataInterface() {propertyValues {
       XCurve {keys [
         <at 0 1.8785 RCIM_Cubic>
         {arriveTangent=-1.3785, interpMode=RCIM_Cubic, leaveTangent=-1.3785, time=0.15, value=0.9996}
         <at 1 0.5 RCIM_Cubic>
        ]}
       YCurve {keys [<at 0 2 RCIM_Cubic>, {arriveTangent=-1.5, interpMode=RCIM_Cubic, leaveTangent=-1.5, time=0.15, value=1}, <at 1 0.5 RCIM_Cubic>]}
       ZCurve {keys [
         <at 0 1.8805 RCIM_Cubic>
         {arriveTangent=-1.3562, interpMode=RCIM_Cubic, leaveTangent=-1.3562, time=0.15, value=0.916}
         <at 1 0.5243 RCIM_Cubic>
        ]}
      }})
    ScaleA=true
    ScaleRGB=true
   )
   &FN_UpdateAnimTrail(
    Locked Particle Color:LinearColor=&MultiplyLinearColor(
     Linear Color=$Particles.Color
     Scale Factor=&MakeFloat4FromVectorAndFloat(
      "Vector (RGB)"=&VectorFromCurve(VectorCurve=DataInterface() {propertyValues {
         XCurve {keys [
           <at 0 1.8785 RCIM_Cubic>
           {arriveTangent=-1.3785, interpMode=RCIM_Cubic, leaveTangent=-1.3785, time=0.15, value=0.9996}
           <at 1 0.5 RCIM_Cubic>
          ]}
         YCurve {keys [<at 0 2 RCIM_Cubic>, {arriveTangent=-1.5, interpMode=RCIM_Cubic, leaveTangent=-1.5, time=0.15, value=1}, <at 1 0.5 RCIM_Cubic>]}
         ZCurve {keys [
           <at 0 1.8805 RCIM_Cubic>
           {arriveTangent=-1.3562, interpMode=RCIM_Cubic, leaveTangent=-1.3562, time=0.15, value=0.916}
           <at 1 0.5243 RCIM_Cubic>
          ]}
        }})
      "Float (Alpha)"=&Multiply_Float(
       A=&FloatFromCurve(FloatCurve=DataInterface() {propertyValues {Curve {keys [
            <at 0 0 RCIM_Cubic>
            {arriveTangent=0.6534, interpMode=RCIM_Cubic, leaveTangent=0.6534, time=0.1871, value=1.0162}
            {arriveTangent=-1.2501, interpMode=RCIM_Cubic, leaveTangent=-1.2501, time=0.3318, value=0.2168}
            <at 1 0 RCIM_Cubic>
           ]}}})
       B=$System.FN_Integrator001.Value
      )
     )
    )
    Locked Particle RibbonWidth=1
    SkeletalMesh=$User.SkeletalMesh
   ) {module=FN_UpdateAnimTrail001}
   &ScaleColor(
    Scale Alpha=&FloatFromCurve(FloatCurve=DataInterface() {propertyValues {Curve {keys [<at 0 0>,<at 1 1>]}}}, CurveIndex=$System.FN_Integrator001.NormalizedValue)
    ScaleA=true
    ScaleRGB=false
   ) {module=ScaleColor001}
   /SetVariables(Particles.RibbonU0Override=$Particles.NormalizedAge) {isSetParameters=true}
   &SolveForcesAndVelocity(Clamp Velocity=false, Limit Acceleration=false)
  ]
  rendererClasses [&NiagaraRibbonRendererProperties]
  renderers [{rendererClass=&NiagaraRibbonRendererProperties}]
 }
}
summary {
 systemName=NS_GliderContrail_Strawberry
 emitters [
  {emitterName=WingTrail, rendererClasses [&NiagaraRibbonRendererProperties]}
  {emitterName=Wind, rendererClasses [&NiagaraMeshRendererProperties]}
  {emitterName=Flares, rendererClasses [&NiagaraSpriteRendererProperties]}
  {emitterName=Engines, rendererClasses [&NiagaraMeshRendererProperties]}
  {emitterName=FireBig, rendererClasses [&NiagaraSpriteRendererProperties]}
  {emitterName=Light, rendererClasses [&NiagaraLightRendererProperties]}
 ]
 userVariables [
  User.bIsFrontEnd {type=NiagaraBool}
  User.bIsFrontEndPreview {type=NiagaraBool}
  User.bIsFullyDeployed {type=NiagaraBool}
  User.ForwardDot {type=NiagaraFloat, defaultValue=0, description=""}
  User.RibbonIDOffset {type=NiagaraInt32}
  User.RightDot {type=NiagaraFloat, defaultValue=0, description=""}
  User.SkeletalMesh {
   type=NiagaraDataInterfaceSkeletalMesh
   defaultValue {
    struct=NiagaraExt_VariableValue_DataInterface
    value {dataInterface=&self:NiagaraDataInterfaceSkeletalMesh_0}
   }
  }
  User.SKM_Engines {
   type=NiagaraDataInterfaceSkeletalMesh
   defaultValue {
    struct=NiagaraExt_VariableValue_DataInterface
    value {dataInterface=&self:NiagaraDataInterfaceSkeletalMesh_2}
   }
  }
  User.SKM_Wind {
   type=NiagaraDataInterfaceSkeletalMesh
   defaultValue {
    struct=NiagaraExt_VariableValue_DataInterface
    value {dataInterface=&self:NiagaraDataInterfaceSkeletalMesh_1}
   }
  }
 ]
}
