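# Niagara deploy effect for the Strawberry Pilot glider, drawing ribbon trails from bones or sockets of the User.SkeletalMesh.
# Point_Trail spawns delayed ribbons at alternate socket indices, split by spawn group.
# Trail_Wide and Trail_Soft each build two anim trails between index pairs, fading spawn alpha and scaling width over emitter loop age, with shared color-over-life curves.
# Particle_Emitter001, a single-socket FX_Trail_L_01 variant of the point trail, is disabled.
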NiagaraSystem /BRCosmetics/Gadgets/Assets/VinderTech_GliderChute/Glider_Strawberry_Pilot/FX/NS_Glider_Strawberry_Deploy
dependencies {
 usedDataInterfaces [&NiagaraDataInterfaceVectorCurve, &NiagaraDataInterfaceCurve]
 usedDynamicInputs []
 usedModules [&SystemState]
}
emitters {
 Particle_Emitter001 {
  bEnabled=false
  emitterUpdateScript [
   &EmitterState(
    Inactive Response=::NewEnumerator1
    Life Cycle Mode=::NewEnumerator1
    Loop Behavior=::NewEnumerator1
    Loop Delay=0.25
    Loop Duration=1
    UseLoopDelay=true
   )
   &SpawnRate(SpawnRate=20, Use Spawn Probability=false)
  ]
  particleSpawnScript [
   &InitializeRibbon(
    Lifetime=0.5
    Module.Write Color=true
    Module.Write Lifetime=true
    Module.Write Mass=true
    Module.Write Material Random=true
    Module.Write Position=false
    Module.Write RibbonTwist=false
    Module.Write RibbonWidth=true
    Ribbon Width=10
   )
   &SkeletalMeshLocation(
    Inherit Velocity=false
    MaskSpawnGroup=false
    Mesh Sampling Type=::NewEnumerator5
    Skeletal Mesh=DataInterface() {propertyValues {
      DefaultMesh=null
      ExcludeBoneName=root
      OptionalComponentClass=null
      PreviewMesh=null
      SkinningMode=SkinOnTheFly
      SoftSourceActor=null
      SourceMode=Default
      UvSetIndex=0
      WholeMeshLOD=-1
      bExcludeBone=false
      bReadDeformedGeometry=true
      bRequireCurrentFrameData=true
      ComponentTags []
      FilteredBones []
      FilteredSockets [FX_Trail_L_01]
      MeshUserParameter {parameter {name=None, typeDefHandle {registeredTypeIndex=1}, varData []}}
      SamplingRegions []
     }}
    Socket Sampling Mode=::NewEnumerator5
   )
   &Color(Color=&MakeLinearColorFromVectorAndFloat(
     "Vector (RGB)"=&VectorFromCurve(VectorCurve=DataInterface() {propertyValues {XCurve {keys [<at 0.75 1>,<at 1 0>]}, YCurve {keys [<at 0.75 1>,<at 1 0>]}, ZCurve {keys [<at 0.75 1>,<at 1 0>]}}})
     "Float (Alpha)"=&ClampFloat(Float=&FloatFromCurve(FloatCurve=DataInterface() {propertyValues {Curve {keys [<at 0.5 1>,<at 1 0>]}}}))
    ))
  ]
  particleUpdateScript [
   &ParticleState()
   &ScaleColor(
    Scale Alpha=&FloatFromCurve(FloatCurve=DataInterface() {propertyValues {Curve {keys [
         <at 0 0 RCIM_Cubic>
         {arriveTangent=0.6534, interpMode=RCIM_Cubic, leaveTangent=0.6534, time=0.1871, value=1.0162}
         {arriveTangent=-1.2501, interpMode=RCIM_Cubic, leaveTangent=-1.2501, time=0.3318, value=0.2168}
         <at 1 0 RCIM_Cubic>
        ]}}})
    Scale RGB=&VectorFromCurve(VectorCurve=DataInterface() {propertyValues {
       XCurve {keys [<at 0 2 RCIM_Cubic>, {arriveTangent=-1.5, interpMode=RCIM_Cubic, leaveTangent=-1.5, time=0.6, value=0.9996}, <at 1 0.5 RCIM_Cubic>]}
       YCurve {keys [<at 0 0.5551 RCIM_Cubic>, {arriveTangent=-0.0551, interpMode=RCIM_Cubic, leaveTangent=-0.0551, time=0.6, value=1}, <at 1 0.5 RCIM_Cubic>]}
       ZCurve {keys [
         <at 0 1.5825 RCIM_Cubic>
         {arriveTangent=-1.0582, interpMode=RCIM_Cubic, leaveTangent=-1.0582, time=0.6, value=0.916}
         <at 1 0.5243 RCIM_Cubic>
        ]}
      }})
    ScaleA=true
    ScaleRGB=true
   )
  ]
  rendererClasses [&NiagaraRibbonRendererProperties]
  renderers [{rendererClass=&NiagaraRibbonRendererProperties}]
 }
 Point_Trail {
  emitterUpdateScript [
   &EmitterState(
    Inactive Response=::NewEnumerator1
    Life Cycle Mode=::NewEnumerator1
    Loop Behavior=::NewEnumerator1
    Loop Delay=0.25
    Loop Duration=1
    UseLoopDelay=true
   )
   &SpawnRate(SpawnRate=20, Use Spawn Probability=false)
  ]
  particleSpawnScript [
   &FN_SetSpawnGroup(Number of Spawn Groups=2)
   &InitializeRibbon(
    Lifetime=0.5
    Module.Write Color=true
    Module.Write Lifetime=true
    Module.Write Mass=true
    Module.Write Material Random=true
    Module.Write Position=false
    Module.Write RibbonTwist=false
    Module.Write RibbonWidth=true
    Ribbon Width=15
   )
   &SkeletalMeshLocation(
    Bone / Socket Index=&Multiply_Int(A=$Particles.SpawnGroup, B=2)
    Inherit Velocity=false
    MaskSpawnGroup=false
    Mesh Sampling Type=::NewEnumerator5
    Skeletal Mesh=$User.SkeletalMesh
    Socket Sampling Mode=::NewEnumerator5
   )
   &Color(Color=&MakeLinearColorFromVectorAndFloat(
     "Vector (RGB)"=&VectorFromCurve(VectorCurve=DataInterface() {propertyValues {XCurve {keys [<at 0.75 1>,<at 1 0>]}, YCurve {keys [<at 0.75 1>,<at 1 0>]}, ZCurve {keys [<at 0.75 1>,<at 1 0>]}}})
     "Float (Alpha)"=&ClampFloat(Float=&FloatFromCurve(FloatCurve=DataInterface() {propertyValues {Curve {keys [<at 0.85 1>,<at 1 0>]}}}))
    ))
  ]
  particleUpdateScript [
   &ParticleState()
   &ScaleColor(
    Scale Alpha=&FloatFromCurve(FloatCurve=DataInterface() {propertyValues {Curve {keys [
         <at 0 0 RCIM_Cubic>
         {arriveTangent=0.6534, interpMode=RCIM_Cubic, leaveTangent=0.6534, time=0.1871, value=1.0162}
         {arriveTangent=-1.2501, interpMode=RCIM_Cubic, leaveTangent=-1.2501, time=0.3318, value=0.2168}
         <at 1 0 RCIM_Cubic>
        ]}}})
    Scale RGB=&VectorFromCurve(VectorCurve=DataInterface() {propertyValues {
       XCurve {keys [<at 0 2 RCIM_Cubic>, {arriveTangent=-1.5, interpMode=RCIM_Cubic, leaveTangent=-1.5, time=0.6, value=0.9996}, <at 1 0.5 RCIM_Cubic>]}
       YCurve {keys [<at 0 0.5551 RCIM_Cubic>, {arriveTangent=-0.0551, interpMode=RCIM_Cubic, leaveTangent=-0.0551, time=0.6, value=1}, <at 1 0.5 RCIM_Cubic>]}
       ZCurve {keys [
         <at 0 1.5825 RCIM_Cubic>
         {arriveTangent=-1.0582, interpMode=RCIM_Cubic, leaveTangent=-1.0582, time=0.6, value=0.916}
         <at 1 0.5243 RCIM_Cubic>
        ]}
      }})
    ScaleA=true
    ScaleRGB=true
   )
  ]
  rendererClasses [&NiagaraRibbonRendererProperties]
  renderers [{rendererClass=&NiagaraRibbonRendererProperties}]
 }
 Trail_Soft {
  emitterSpawnScript [&FN_RibbonIDOffsetUtility()]
  emitterUpdateScript [
   &EmitterState()
   &FN_SpawnAnimTrail(Number of Anim Trails=2)
  ]
  particleSpawnScript [
   &FN_InitializeAnimTrailBoneIndecies(First Bone or Socket Index=1, Second Bone or Socket Index=0)
   &FN_InitializeAnimTrailBoneIndecies(AnimTrail Selection=1, First Bone or Socket Index=3, Second Bone or Socket Index=2) {module=FN_InitializeAnimTrailBoneIndecies001}
   &FN_InitializeAnimTrail(
    Color=&MakeLinearColor(
     R:NiagaraFloat=1
     G:NiagaraFloat=1
     B:NiagaraFloat=1
     A:NiagaraFloat=&FloatFromCurve(FloatCurve=DataInterface() {propertyValues {Curve {keys [<at 0.2 1>,<at 0.35 0>]}}}, CurveIndex=$Emitter.NormalizedLoopAge, Scale Curve=2)
    )
    Lifetime=1.2
   )
  ]
  particleUpdateScript [
   &ParticleState()
   &ScaleColor(
    Scale Alpha=&FloatFromCurve(FloatCurve=DataInterface() {propertyValues {Curve {keys [
         <at 0 0 RCIM_Cubic>
         {arriveTangent=0.6534, interpMode=RCIM_Cubic, leaveTangent=0.6534, time=0.1871, value=1.0162}
         {arriveTangent=-1.2501, interpMode=RCIM_Cubic, leaveTangent=-1.2501, time=0.3318, value=0.2168}
         <at 1 0 RCIM_Cubic>
        ]}}})
    Scale RGB=&VectorFromCurve(VectorCurve=DataInterface() {propertyValues {
       XCurve {keys [<at 0 2 RCIM_Cubic>, {arriveTangent=-1.5, interpMode=RCIM_Cubic, leaveTangent=-1.5, time=0.6, value=0.9996}, <at 1 0.5 RCIM_Cubic>]}
       YCurve {keys [<at 0 0.5551 RCIM_Cubic>, {arriveTangent=-0.0551, interpMode=RCIM_Cubic, leaveTangent=-0.0551, time=0.6, value=1}, <at 1 0.5 RCIM_Cubic>]}
       ZCurve {keys [
         <at 0 1.5825 RCIM_Cubic>
         {arriveTangent=-1.0582, interpMode=RCIM_Cubic, leaveTangent=-1.0582, time=0.6, value=0.916}
         <at 1 0.5243 RCIM_Cubic>
        ]}
      }})
    ScaleA=true
    ScaleRGB=true
   ) {module=ScaleColor001}
   &ScaleRibbonWidth(Ribbon Width Scale=&FloatFromCurve(FloatCurve=DataInterface() {propertyValues {Curve {keys [<at 0 0.9>,<at 1 1>]}}}, CurveIndex=$Emitter.NormalizedLoopAge))
   &FN_UpdateAnimTrail(
    Locked Particle Color:LinearColor=$Particles.Color
    Locked Particle RibbonWidth:NiagaraFloat=$Particles.RibbonWidth
    SkeletalMesh=$User.SkeletalMesh
   )
  ]
  rendererClasses [&NiagaraRibbonRendererProperties]
  renderers [{rendererClass=&NiagaraRibbonRendererProperties}]
 }
 Trail_Wide {
  emitterSpawnScript [&FN_RibbonIDOffsetUtility()]
  emitterUpdateScript [
   &EmitterState()
   &FN_SpawnAnimTrail(Number of Anim Trails=2)
  ]
  particleSpawnScript [
   &FN_InitializeAnimTrailBoneIndecies(First Bone or Socket Index=1, Second Bone or Socket Index=0)
   &FN_InitializeAnimTrailBoneIndecies(AnimTrail Selection=1, First Bone or Socket Index=3, Second Bone or Socket Index=2) {module=FN_InitializeAnimTrailBoneIndecies001}
   &FN_InitializeAnimTrail(
    Color=&MakeLinearColor(
     R:NiagaraFloat=1
     G:NiagaraFloat=1
     B:NiagaraFloat=1
     A:NiagaraFloat=&FloatFromCurve(
      FloatCurve=DataInterface() {propertyValues {Curve {keys [{arriveTangent=-5.2069, interpMode=RCIM_Cubic, leaveTangent=-5.2069, tangentMode=RCTM_User, time=0.1, value=1}, <at 0.8 0>]}}}
      CurveIndex=$Emitter.NormalizedLoopAge
     )
    )
    Lifetime=0.35
   )
  ]
  particleUpdateScript [
   &ParticleState()
   &ScaleColor(
    Scale Alpha=&FloatFromCurve(FloatCurve=DataInterface() {propertyValues {Curve {keys [
         <at 0 0 RCIM_Cubic>
         {arriveTangent=0.6534, interpMode=RCIM_Cubic, leaveTangent=0.6534, time=0.1871, value=1.0162}
         {arriveTangent=-1.2501, interpMode=RCIM_Cubic, leaveTangent=-1.2501, time=0.3318, value=0.2168}
         <at 1 0 RCIM_Cubic>
        ]}}})
    Scale RGB=&VectorFromCurve(VectorCurve=DataInterface() {propertyValues {
       XCurve {keys [<at 0 2 RCIM_Cubic>, {arriveTangent=-1.5, interpMode=RCIM_Cubic, leaveTangent=-1.5, time=0.6, value=0.9996}, <at 1 0.5 RCIM_Cubic>]}
       YCurve {keys [<at 0 0.5551 RCIM_Cubic>, {arriveTangent=-0.0551, interpMode=RCIM_Cubic, leaveTangent=-0.0551, time=0.6, value=1}, <at 1 0.5 RCIM_Cubic>]}
       ZCurve {keys [
         <at 0 1.5825 RCIM_Cubic>
         {arriveTangent=-1.0582, interpMode=RCIM_Cubic, leaveTangent=-1.0582, time=0.6, value=0.916}
         <at 1 0.5243 RCIM_Cubic>
        ]}
      }})
    ScaleA=true
    ScaleRGB=true
   ) {module=ScaleColor001}
   &ScaleRibbonWidth(Ribbon Width Scale=&FloatFromCurve(FloatCurve=DataInterface() {propertyValues {Curve {keys [<at 0 0.7>,<at 1 1>]}}}, CurveIndex=$Emitter.NormalizedLoopAge))
   &ScaleRibbonWidth(Ribbon Width Scale=&FloatFromCurve(FloatCurve=DataInterface() {propertyValues {Curve {keys [<at 0 0.6 RCIM_Cubic>, {arriveTangent=0.2, interpMode=RCIM_Cubic, time=0.3, value=0.8}, <at 1 0.5 RCIM_Cubic>]}}})) {module=ScaleRibbonWidth001}
   &FN_UpdateAnimTrail(
    Locked Particle Color:LinearColor=$Particles.Color
    Locked Particle RibbonWidth:NiagaraFloat=$Particles.RibbonWidth
    SkeletalMesh=$User.SkeletalMesh
   )
  ]
  rendererClasses [&NiagaraRibbonRendererProperties]
  renderers [{rendererClass=&NiagaraRibbonRendererProperties}]
 }
}
summary {
 systemName=NS_Glider_Strawberry_Deploy
 emitters [
  {emitterName=Point_Trail, rendererClasses [&NiagaraRibbonRendererProperties]}
  {bEnabled=false, emitterName=Particle_Emitter001, rendererClasses [&NiagaraRibbonRendererProperties]}
  {emitterName=Trail_Wide, rendererClasses [&NiagaraRibbonRendererProperties]}
  {emitterName=Trail_Soft, rendererClasses [&NiagaraRibbonRendererProperties]}
 ]
 userVariables [
  User.RibbonIDOffset {type=NiagaraInt32}
  User.SkeletalMesh {
   type=NiagaraDataInterfaceSkeletalMesh
   defaultValue {
    struct=NiagaraExt_VariableValue_DataInterface
    value {dataInterface=&self:NiagaraDataInterfaceSkeletalMesh_1}
   }
  }
 ]
}
